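; UI director component that manages HUD scenes for vehicle mods: a marker scene, weapon-mod equipped and unequipped scenes, and a Slinky scene.
; In zone, vehicle-state and weapon-equip events wait one tick for the seat's cached weapon definition, then update the cached vehicle and seat-change binding.
; UpdateVehicleModStatus matches seat weapon tags and checks for seat zero on Slinky vehicles, setting flags that drive scene add and remove.
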(event EventOnReadyToShowUI
  (bind _returnvalue (BlueprintContext|GetFortHUDContext))
  (CallFunction|AddScene (Variables|Default|GetVehicleModsMarkerScene))
  (if (GlobalUIContext|IsinZone (PlayerController|LocalPlayerSubsystems|GetFortGlobalUIContext (CallFunction|GetOwningLocalPlayerController)))
    (HUD|BindEventtoOnPlayerVehicleStateChanged _returnvalue (delegate Custom|OnPlayerVehicleStateChangedEvent))
    ;[0.1407 0.15 0 0.5] Note on Delay Until Next Frame:
    ; The CachedWeaponDef of the VehicleWeaponSeatComponent is created and assigned during the callstack the weapon mod is equipped or when the player enters the seat. This CachedWeaponDef is used inside UpdateVehicleModStatus.
    ;
    ; This results in bPlayerInWeaponModSeat to be false, because we never had a valid CachedWeaponDef object to referance.
    ;
    ; The delay ensures that the VehicleWeaponSeatComponent has setup the weapon before we try to access it in the various delegates around here.
    (HUD|BindEventtoOnWeaponEquippedDelegate _returnvalue (delegate Custom|On%20Weapon%20Equipped))))

(event Custom|On%20Weapon%20Equipped
  (params (NewWeapon :type Fort%20Weapon%20Object%20Reference) (PrevWeapon :type Fort%20Weapon%20Object%20Reference))
  ;[0.1407 0.15 0 0.5] Note on Delay Until Next Frame:
  ; The CachedWeaponDef of the VehicleWeaponSeatComponent is created and assigned during the callstack the weapon mod is equipped or when the player enters the seat. This CachedWeaponDef is used inside UpdateVehicleModStatus.
  ;
  ; This results in bPlayerInWeaponModSeat to be false, because we never had a valid CachedWeaponDef object to referance.
  ;
  ; The delay ensures that the VehicleWeaponSeatComponent has setup the weapon before we try to access it in the various delegates around here.
  (Utilities|FlowControl|DelayUntilNextTick)
  (Default|OnVehicleStateChanged 0 (Variables|Default|GetCachedVehicle))
  (Utilities|Casting|CastToFortWeaponRangedForVehicleMod NewWeapon
    (:then
      (Variables|Default|SetbVehicleModWeaponEquipped true))
    (:CastFailed
      (Variables|Default|SetbVehicleModWeaponEquipped false)))
  (Default|AddVehicleWeaponModScene))

(event Custom|OnPlayerVehicleStateChangedEvent
  (params (PlayerPawn :type Fort%20Player%20Pawn%20Object%20Reference) (NewVehicle :type Actor%20Object%20Reference) (OldVehicle :type Actor%20Object%20Reference))
  ;[0.1407 0.15 0 0.5] Note on Delay Until Next Frame:
  ; The CachedWeaponDef of the VehicleWeaponSeatComponent is created and assigned during the callstack the weapon mod is equipped or when the player enters the seat. This CachedWeaponDef is used inside UpdateVehicleModStatus.
  ;
  ; This results in bPlayerInWeaponModSeat to be false, because we never had a valid CachedWeaponDef object to referance.
  ;
  ; The delay ensures that the VehicleWeaponSeatComponent has setup the weapon before we try to access it in the various delegates around here.
  (Utilities|FlowControl|DelayUntilNextTick)
  (Default|OnVehicleStateChanged PlayerPawn NewVehicle OldVehicle))

(event EventEndPlay
  (params (EndPlayReason :type EEndPlayReason%20Enum :default "Destroyed"))
  (CallFunction|RemoveScene (Variables|Default|GetVehicleModsMarkerScene)))

(fn AddVehicleWeaponModScene
  (Default|RemoveVehicleModScenes)
  (if (Variables|Default|GetbPlayerInWeaponModSeat)
    (if (Variables|Default|GetbVehicleModWeaponEquipped)
      (if (not (Variables|Default|GetbAddedEquippedScene))
        (CallFunction|AddScene (Variables|Default|GetEquippedScene))
        (Variables|Default|SetbAddedEquippedScene true))
      (elif (Variables|Default|GetbAddedUnequippedScene)
        _
        (else
          (CallFunction|AddScene (Variables|Default|GetUnequippedScene))
          (Variables|Default|SetbAddedUnequippedScene true))))))

(fn OnVehicleStateChanged
  (params (PlayerPawn :type Fort%20Player%20Pawn%20Object%20Reference) (NewVehicle :type Actor%20Object%20Reference) (OldVehicle :type Actor%20Object%20Reference))
  (bind _cachedvehicle (Variables|Default|GetCachedVehicle))
  (Utilities|IsValid _cachedvehicle
    (:Is%20Valid
      (bind _returnvalue (Vehicle|GetVehicleSeatComponent _cachedvehicle))
      (Utilities|IsValid _returnvalue
        (:Is%20Valid
          (Vehicle|UnbindEventfromOnVehicleSeatPlayerSlotsChanged _returnvalue (EventDispatchers|CreateEvent :!event Update%20Vehicle%20Mod%20Status)))
        (:Is%20Not%20Valid
          _))
      (bind _asfort_athena_vehicle (Utilities|Casting|CastToFortAthenaVehicle NewVehicle)
        (:then
          (bind _output_get (Variables|Default|SetCachedVehicle _asfort_athena_vehicle))
          (Utilities|IsValid _output_get
            (:Is%20Valid
              (bind _returnvalue_2 (Vehicle|GetVehicleSeatComponent _output_get))
              (Utilities|IsValid _returnvalue_2
                (:Is%20Valid
                  (Vehicle|BindEventtoOnVehicleSeatPlayerSlotsChanged _returnvalue_2 (EventDispatchers|CreateEvent :!event Update%20Vehicle%20Mod%20Status)))
                (:Is%20Not%20Valid
                  _))
              (Default|UpdateVehicleModStatus))
            (:Is%20Not%20Valid
              _)))
        (:CastFailed
          _)))
    (:Is%20Not%20Valid
      _)))

(fn RemoveVehicleModScenes
  (if (Variables|Default|GetbAddedEquippedScene)
    (CallFunction|RemoveScene (Variables|Default|GetEquippedScene))
    (Variables|Default|SetbAddedEquippedScene false)
    (else
      _))
  (if (Variables|Default|GetbAddedUnequippedScene)
    (CallFunction|RemoveScene (Variables|Default|GetUnequippedScene))
    (Variables|Default|SetbAddedUnequippedScene false)))

(fn UpdateVehicleModStatus
  (bind _returnvalue (CallFunction|GetOwningLocalPlayerController))
  (bind _returnvalue_2 (Pawn|GetViewTarget _returnvalue))
  (Utilities|IsValid _returnvalue_2
    (:Is%20Valid
      (bind _asfort_player_pawn (Utilities|Casting|CastToFortPlayerPawn _returnvalue_2)
        (:then
          (bind _asfort_vehicle_interface (Utilities|Casting|CastToFortVehicleInterface (Variables|Default|GetCachedVehicle))
            (:then
              (Utilities|FlowControl|Sequence
                (:then_0
                  (Utilities|IsValid (VehicleWeapon|GetSeatWeaponComponent _asfort_vehicle_interface (Vehicle|GetVehicleSeatIndex _asfort_player_pawn))
                    (:Is%20Valid
                      (Utilities|IsValid (Variables|/Script/FortniteGame.FortVehicleSeatWeaponComponent|GetCachedWeaponDef (VehicleWeapon|GetSeatWeaponComponent _asfort_vehicle_interface (Vehicle|GetVehicleSeatIndex _asfort_player_pawn)))
                        (:Is%20Valid
                          (if (GameplayTags|HasAnyMatchingGameplayTags (Variables|/Script/FortniteGame.FortVehicleSeatWeaponComponent|GetCachedWeaponDef (VehicleWeapon|GetSeatWeaponComponent _asfort_vehicle_interface (Vehicle|GetVehicleSeatIndex _asfort_player_pawn))) (Variables|Default|GetVehicleModTags))
                            (Variables|Default|SetbPlayerInWeaponModSeat true)
                            (else
                              (Variables|Default|SetbPlayerInWeaponModSeat false)))
                          (Default|AddVehicleWeaponModScene))
                        (:Is%20Not%20Valid
                          _)))
                    (:Is%20Not%20Valid
                      _)))
                (:then_1
                  (Utilities|IsValid (Actor|GetComponentbyClass (Variables|Default|GetCachedVehicle) "/VehicleMods/Mods/Slinky/VMC_Slinky.VMC_Slinky_C")
                    (:Is%20Valid
                      (Variables|Default|SetPlayerInSlinkyModSeat (== (Vehicle|GetVehicleSeatIndex _asfort_player_pawn) 0))
                      (Default|UpdateSlinkyModScene))))))
            (:CastFailed
              _)))
        (:CastFailed
          _)))
    (:Is%20Not%20Valid
      _)))

(fn UpdateScene
  (Utilities|FlowControl|Sequence
    (:then_0
      (if (Variables|Default|GetbPlayerInWeaponModSeat)
        (if (Variables|Default|GetbVehicleModWeaponEquipped)
          _)
        (else
          (Default|RemoveVehicleModScenes))))
    (:then_1
      (Default|UpdateSlinkyModScene))))

(fn UpdateSlinkyModScene
  (if (Variables|Default|GetAddedSlinkyModScene)
    (CallFunction|RemoveScene "/VehicleModsUI/Scenes/UIScene_VehicleMods_Slinky.UIScene_VehicleMods_Slinky")
    (Variables|Default|SetAddedSlinkyModScene false)
    (else
      _))
  (if (Variables|Default|GetPlayerInSlinkyModSeat)
    (CallFunction|AddScene "/VehicleModsUI/Scenes/UIScene_VehicleMods_Slinky.UIScene_VehicleMods_Slinky")
    (Variables|Default|SetAddedSlinkyModScene true)))
